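# Sound cue for a generic short weapon destroy sound, built on a single pistol destroy sample.
# A fixed short Delay node holds back the Wave Player, so the sample starts slightly after the cue triggers.
# It borrows small melee weapon attenuation and routes to the ranged weapon low-ammo/degraded sound class.
# A concurrency override blocks new instances once the limit is reached, leaving playing ones intact.
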SoundCue /Game/Sounds/Fort_Weapons/Destroys/Generic_Short_Destroy_Cue
root=Delay_3
nodes [
 {id=Delay_3, children [WavePlayer_7], properties {delayMax=0.3, delayMin=0.3}}
 {id=WavePlayer_7, properties {soundWaveAssetPtr=/Game/Sounds/Fort_Weapons/pistol/degrades/pistol_destroy_04}}
]
properties {
 attenuationSettings=&Fort_Weapon_Melee_Sml_Att
 bOverrideConcurrency=true
 maxDistance=2800
 soundClassObject=/Game/Sounds/FortSoundClassesAndModes/Fort_WeaponRanged_LowAmmo_Degraded_PSM
 volumeMultiplier=1
 concurrencyOverrides {resolutionRule=PreventNew}
}
